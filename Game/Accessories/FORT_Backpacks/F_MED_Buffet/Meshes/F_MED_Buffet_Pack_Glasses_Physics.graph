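# Physics asset for the glasses on the F_MED_Buffet backpack cosmetic.
# It has two small capsule bodies, on the Dyn_Glasses and Dyn_Master_Glasses bones, each offset along its bone's X axis.
# A single constraint joins the two bodies with both swing axes and twist locked, so neither can rotate relative to the other.
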PhysicsAsset /Game/Accessories/FORT_Backpacks/F_MED_Buffet/Meshes/F_MED_Buffet_Pack_Glasses_Physics
bodies {
 Dyn_Glasses [{shapeType=Capsule, length=1, radius=1, shapeName=Dyn_Glasses_capsule, center=<0.094 0 0>}]
 Dyn_Master_Glasses [{shapeType=Capsule, length=1, radius=1, shapeName=Dyn_Master_Glasses_capsule, center=<2.7729 0 0>}]
}
constraints [{
  bone1Name=Dyn_Glasses
  bone2Name=Dyn_Master_Glasses
  swing1LimitDegrees=45
  swing1Motion=Locked
  swing2LimitDegrees=45
  swing2Motion=Locked
  twistLimitDegrees=45
  twistMotion=Locked
 }]
