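# Physics asset for the F_MED_Buffet backpack, simulating a dangling creature-like figure with head, ears, nose, feet and tail.
# Capsule bodies on Dyn_ bones branch from Dyn_Master_Body through upper-body and lower-body masters to the head, pelvis, leg and tail chains.
# Master upper-body, lower-body and nose joints rotate freely, while most other joints limit swing and nearly eliminate twist.
# The nose is fully rotation-locked to its master bone.
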PhysicsAsset /Game/Accessories/FORT_Backpacks/F_MED_Buffet/Meshes/F_MED_Buffet_Pack_Physics
bodies {
 Dyn_Body1 [{shapeType=Capsule, length=5, radius=5, shapeName=Dyn_Body1_capsule, center=<0.4209 -2.5642 0>}]
 Dyn_Head [{shapeType=Capsule, length=5, radius=3, shapeName=Dyn_Head_capsule, center=<-1.0366 -3.5092 0>, rotation=<rot 90 14.0363 -48.3683>}]
 Dyn_Left_Ear1 [{
   shapeType=Capsule
   length=2.8613
   radius=0.8295
   shapeName=Dyn_Left_Ear1_capsule
   center=<2.0818 -0.4528 -0.1484>
   rotation=<rot -0.1789 -107.2298 -87.9551>
  }]
 Dyn_Left_Foot [{shapeType=Capsule, length=2, radius=1, shapeName=Dyn_Left_Foot_capsule, center=<1 0 0>, rotation=<rot 90 0 0>}]
 Dyn_Left_Foot1 [{shapeType=Capsule, length=2, radius=1, shapeName=Dyn_Left_Foot1_capsule, center=<1 0 0>, rotation=<rot 90 0 0>}]
 Dyn_Master_Body [{shapeType=Capsule, length=0.505, radius=0.505, shapeName=Dyn_Master_Body_capsule, rotation=<rot 0 -90 0>}]
 Dyn_Master_LowerBody [{shapeType=Capsule, length=5, radius=3, shapeName=Dyn_Master_LowerBody_capsule, center=<-0.8729 -0.0313 0.0001>}]
 Dyn_Master_Nose [{shapeType=Capsule, length=0.505, radius=1.255, shapeName=Dyn_Master_Nose_capsule, center=<-0.169 0 0>, rotation=<rot 0 -90 0>}]
 Dyn_Master_UpperBody [{shapeType=Capsule, length=0.505, radius=0.505, shapeName=Dyn_Master_UpperBody_capsule, rotation=<rot 0 -90 0>}]
 Dyn_Nose [{shapeType=Capsule, length=1, radius=1, shapeName=Dyn_Nose_capsule, center=<-0.0131 0.2573 0.0183>, rotation=<rot 1.9763 -90 -3.7651>}]
 Dyn_Pelvis_ [{shapeType=Capsule, length=5, radius=3, shapeName=Dyn_Pelvis__capsule, center=<1 -1.2565 0>}]
 Dyn_Right_Ear1 [{
   shapeType=Capsule
   length=2.8561
   radius=0.9011
   shapeName=Dyn_Right_Ear1_capsule
   center=<2.281 0.7412 -0.7452>
   rotation=<rot -1.2989 -82.4706 -74.9717>
  }]
 Dyn_Right_Foot [{shapeType=Capsule, length=2, radius=1, shapeName=Dyn_Right_Foot_capsule, center=<1 0 0>, rotation=<rot 90 0 0>}]
 Dyn_Right_Foot1 [{shapeType=Capsule, length=2, radius=1, shapeName=Dyn_Right_Foot1_capsule, center=<1 0 0>, rotation=<rot 90 0 0>}]
 Dyn_Tail1 [{shapeType=Capsule, length=1, radius=1, shapeName=Dyn_Tail1_capsule, center=<0.1213 0 0>, rotation=<rot 90 0 0>}]
 Dyn_Tail2 [{shapeType=Capsule, length=1, radius=1, shapeName=Dyn_Tail2_capsule, center=<0.6845 -0.2473 0>, rotation=<rot 90 -172.875 -152.4384>}]
}
constraints [
 {bone1Name=Dyn_Right_Ear1, bone2Name=Dyn_Head, swing1LimitDegrees=20, swing2LimitDegrees=20, twistLimitDegrees=1}
 {bone1Name=Dyn_Left_Ear1, bone2Name=Dyn_Head, swing1LimitDegrees=20, swing2LimitDegrees=20, twistLimitDegrees=1}
 {bone1Name=Dyn_Body1, bone2Name=Dyn_Master_Body, swing1LimitDegrees=30, swing2LimitDegrees=10, twistLimitDegrees=1}
 {
  bone1Name=Dyn_Master_LowerBody
  bone2Name=Dyn_Master_Body
  swing1LimitDegrees=45
  swing1Motion=Free
  swing2LimitDegrees=45
  swing2Motion=Free
  twistLimitDegrees=45
  twistMotion=Free
 }
 {
  bone1Name=Dyn_Master_UpperBody
  bone2Name=Dyn_Master_Body
  swing1LimitDegrees=45
  swing1Motion=Free
  swing2LimitDegrees=45
  swing2Motion=Free
  twistLimitDegrees=45
  twistMotion=Free
 }
 {bone1Name=Dyn_Head, bone2Name=Dyn_Master_UpperBody, swing1LimitDegrees=45, swing2LimitDegrees=1, twistLimitDegrees=1}
 {bone1Name=Dyn_Right_Foot, bone2Name=Dyn_Pelvis_, swing1LimitDegrees=10, swing2LimitDegrees=20, twistLimitDegrees=1}
 {bone1Name=Dyn_Right_Foot1, bone2Name=Dyn_Right_Foot, swing1LimitDegrees=10, swing2LimitDegrees=20, twistLimitDegrees=1}
 {bone1Name=Dyn_Left_Foot, bone2Name=Dyn_Pelvis_, swing1LimitDegrees=10, swing2LimitDegrees=20, twistLimitDegrees=1}
 {bone1Name=Dyn_Left_Foot1, bone2Name=Dyn_Left_Foot, swing1LimitDegrees=10, swing2LimitDegrees=20, twistLimitDegrees=1}
 {
  bone1Name=Dyn_Nose
  bone2Name=Dyn_Master_Nose
  swing1LimitDegrees=45
  swing1Motion=Locked
  swing2LimitDegrees=45
  swing2Motion=Locked
  twistLimitDegrees=45
  twistMotion=Locked
 }
 {
  bone1Name=Dyn_Master_Nose
  bone2Name=Dyn_Head
  swing1LimitDegrees=45
  swing1Motion=Free
  swing2LimitDegrees=45
  swing2Motion=Free
  twistLimitDegrees=45
  twistMotion=Free
 }
 {bone1Name=Dyn_Pelvis_, bone2Name=Dyn_Master_LowerBody, swing1LimitDegrees=50, swing2LimitDegrees=5, twistLimitDegrees=1}
 {bone1Name=Dyn_Tail1, bone2Name=Dyn_Master_LowerBody, swing1LimitDegrees=45, swing2LimitDegrees=20, twistLimitDegrees=1}
 {bone1Name=Dyn_Tail2, bone2Name=Dyn_Tail1, swing1LimitDegrees=45, swing2LimitDegrees=20, twistLimitDegrees=1}
]
